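; Blueprint for the Jethro CrustyChain pistol that drives its Niagara overheat effect and plays controller rumble when it overheats.
; While heating, a looping timer, slower on mobile, passes normalized heat to OverheatFX, which spawns the effect once past a threshold and updates it.
; The spawn sets weapon and medallion-tag parameters; overheating swaps emitters to cooldown smoke.
; Cooling stops the timer and deactivates the effect; finished cooling, equip and unequip destroy and reset it.
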(event Custom|SpawnFX
  ;[1 0 0.7965 0.5] Spawn Overheat FX
  (bind _returnvalue (Niagara|SpawnSystemAttached "/JethroWeaponGameplay/Assets/FX/NS_Jethro_OverHeat_Base.NS_Jethro_OverHeat_Base" (Variables|Visual|GetWeaponMesh) "None" "0, 0, 0" "0, 0, 0" "KeepRelativeOffset" false false))
  (bind _output_get (Variables|Default|SetSpawned_Overheat_FX _returnvalue))
  (Effects|Components|ParticleSystem|SetEmitterEnable _output_get "CooldownSmokeCones" false)
  (Effects|Components|ParticleSystem|SetEmitterEnable _output_get "CooldownSmoke" false)
  (Niagara|SetNiagaraVariable(Int32) _output_get "WeaponInt" (Variables|Default|GetWeapon%20Int))
  (Niagara|SetNiagaraVariable(Bool) _output_get "HasMedallion" (GameplayTags|HasTag (GameplayTags|GetOwnedGameplayTags (Variables|Default|GetAs%20Fort%20Player%20Pawn) :!class /Script/GameplayTags.BlueprintGameplayTagLibrary) (Variables|Default|GetMedallionTag)))
  (Rendering|SetFirstPersonPrimitiveType _output_get (Variables|/Script/Engine.PrimitiveComponent|GetFirstPersonPrimitiveType (Variables|Visual|GetWeaponMesh))))

(event FortWeapon|EventOnEquippedAndAttached
  (bind _spawned_overheat_fx (Variables|Default|GetSpawned_Overheat_FX))
  ;[0 1 0.4912] On Equip/Unequip -> Reset or else no Overheat FX spawn.
  (|CallParentFunction)
  (CallFunction|Reset)
  (Utilities|IsValid _spawned_overheat_fx
    (:Is%20Valid
      (Components|DestroyComponent _spawned_overheat_fx))))

(event FortWeapon|EventOnUnEquippedAndDetached
  (bind _spawned_overheat_fx (Variables|Default|GetSpawned_Overheat_FX))
  ;[0 1 0.4912] On Equip/Unequip -> Reset or else no Overheat FX spawn.
  (|CallParentFunction)
  (CallFunction|Reset)
  (Utilities|IsValid _spawned_overheat_fx
    (:Is%20Valid
      (Components|DestroyComponent _spawned_overheat_fx))))

(event Custom|OverheatFX
  (params (Overheat%20Val :type Byte))
  (bind _returnvalue (Math|Conversions|ToInteger(Byte) Overheat%20Val))
  (bind _returnvalue_2 (Math|Conversions|ToFloat(Integer) _returnvalue))
  (bind _returnvalue_3 (/ _returnvalue_2 255.0))
  (bind _spawned_overheat_fx (Variables|Default|GetSpawned_Overheat_FX))
  ;[1 0 0.1652] Overheat FX
  (if (> _returnvalue_3 0.25)
    (Utilities|DoOnceThenDo false
      (:DoOnce
        (CallFunction|SpawnFX))
      (:ThenDo
        _))
    (Components|Activation|Activate _spawned_overheat_fx false)
    (Niagara|SetNiagaraVariable(Float) _spawned_overheat_fx "OverHeat" (Math|Float|Clamp(Float) (- (* _returnvalue_3 2.0) 1.0)))))

(event Custom|Reset
  (bind _spawned_overheat_fx (Variables|Default|GetSpawned_Overheat_FX))
  ;[1 0 0.1652] Overheat FX
  (Utilities|DoOnceThenDo false
    (:DoOnce
      (CallFunction|SpawnFX))
    (:ThenDo
      _))
  (Components|Activation|Activate _spawned_overheat_fx false)
  (Niagara|SetNiagaraVariable(Float) _spawned_overheat_fx "OverHeat" (Math|Float|Clamp(Float) (- (* (/ (Math|Conversions|ToFloat(Integer) (Math|Conversions|ToInteger(Byte) (stale Custom|OverheatFX Overheat%20Val))) 255.0) 2.0) 1.0))))

(event EventBeginPlay
  (bind _returnvalue (Game|GetInstigator))
  ;[1 0.6311 0.9916 0.5] Event Begin Play Cast to FortPlayerPawn
  (|CallParentFunction)
  (Utilities|IsValid _returnvalue
    (:Is%20Valid
      (bind _asfort_player_pawn (Utilities|Casting|CastToFortPlayerPawn _returnvalue))
      (Variables|Default|SetAs%20Fort%20Player%20Pawn _asfort_player_pawn))))

(event Overheating|EventOnStartOverheated
  (bind _spawned_overheat_fx (Variables|Default|GetSpawned_Overheat_FX))
  ;[1 0.1083 0] Overheated
  (|CallParentFunction)
  (Effects|Components|ParticleSystem|SetEmitterEnable _spawned_overheat_fx "CooldownSmokeCones" true)
  (Effects|Components|ParticleSystem|SetEmitterEnable _spawned_overheat_fx "CooldownSmoke" true)
  (Effects|Components|ParticleSystem|SetEmitterEnable _spawned_overheat_fx "WeaponMeshArray" false)
  (Effects|Components|ParticleSystem|SetEmitterEnable _spawned_overheat_fx "Ambient_Steam" false)
  (Effects|Components|ParticleSystem|SetEmitterEnable _spawned_overheat_fx "MedallionFlare" false)
  (Game|Feedback|LocalPlayForceFeedback (Fortnite|GetFortPlayerControllerfromActor (Game|GetInstigator)) "/Game/Weapons/ForceFeedback/FF_Jethro_CrustyChain_Overheated.FF_Jethro_CrustyChain_Overheated" false "None"))

(event Overheating|EventOnStartHeating
  ;[1 0.4469 0] Start Heating
  (|CallParentFunction)
  (bind _returnvalue (Utilities|Time|SetTimerbyEvent (EventDispatchers|CreateEvent :!event Custom|Start%20Overheat) (select (Fortnite|IsMobilePlatform) 0.1 0.03) true))
  (Variables|Default|SetOverheatFXTimer _returnvalue))

(event Custom|Start%20Overheat
  ;[1 0.4469 0] Start Heating
  (CallFunction|OverheatFX (Math|Conversions|ToByte(Integer) (Math|Float|Truncate (* 255 (/ (Overheat|GetCurrentOverheatValue) (Overheat|GetOverheatingMaxValue)))))))

(event Overheating|EventOnFinishedCooling
  ;[0 0.6897 1] Finished Cooling
  (|CallParentFunction)
  (Utilities|FlowControl|Delay 0.2)
  (Components|DestroyComponent (Variables|Default|GetSpawned_Overheat_FX))
  (CallFunction|Reset))

(event Overheating|EventOnStartCooling
  ;[0 1 0.8693] Start Cooling
  (|CallParentFunction)
  (Utilities|Time|ClearandInvalidateTimerbyHandle (Variables|Default|GetOverheatFXTimer))
  (Utilities|FlowControl|Delay 0.8)
  (Components|Activation|Deactivate (Variables|Default|GetSpawned_Overheat_FX)))

(event EventOnPlayWeaponFireFX
  (params (bPersistentFire :type Boolean) (bSecondaryFire :type Boolean))
  ;[1 0.9617 0 0.5] On Weapon Fire
  (|CallParentFunction bPersistentFire bSecondaryFire))

(fn ConstructionScript
  (|CallParentFunction))
